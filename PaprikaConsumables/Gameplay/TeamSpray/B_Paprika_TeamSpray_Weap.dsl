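; Team spray weapon Blueprint that drives the bottle liquid and muzzle spray Niagara effects.
; On begin play, equip and ammo change, cosmetic-capable machines set liquid FillPercent from remaining ammo over clip size, optionally reinitializing the system.
; ToggleSprayFX sets EnableSprayVFX on authority only.
; Its OnRep passes IsSpraying to the liquid and activates the spray with reset, unless disabled while already active; it never deactivates it.
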(event EventBeginPlay
  ;[0 0.5221 1] Update fill on initialize
  (CallFunction|UpdateFillPercentage))

(event Custom|UpdateFillPercentage
  (bind _ns_paprika_team_spray_liquid (Variables|Default|GetNS_Paprika_Team_Spray_Liquid))
  ;[0 0.5221 1] Update the fill of the bottle
  (Networking|CanExecuteCosmeticEvents
    (:True
      (Niagara|SetNiagaraVariable(Float) _ns_paprika_team_spray_liquid "FillPercent" (Math|Float|SafeDivide (Math|Conversions|ToFloat(Integer) (FortWeapon|GetRemainingAmmo)) (Math|Conversions|ToFloat(Integer) (FortWeapon|GetBulletsPerClip))))
      (if (Variables|Default|GetShouldReinitialize?)
        (Niagara|ReinitializeSystem _ns_paprika_team_spray_liquid)))))

(event FortWeapon|EventOnEquippedAndAttached
  ;[0 0.5221 1] Fill update timing
  (CallFunction|UpdateFillPercentage))

(event EventOnAmmoCountChanged
  ;[0 0.5221 1] Fill update timing
  (CallFunction|UpdateFillPercentage))

(event Custom|ToggleSprayFX
  (params (Enabled? :type Boolean))
  ;[0.03429 0.3069 0.7899] Set spraying muzzle
  (Networking|SwitchHasAuthority
    (:Authority
      (Variables|Default|SetEnableSprayVFX Enabled?))))

(fn OnRep_EnableSprayVFX
  (bind _ns_paprika_teamspray_spray (Variables|Default|GetNS_Paprika_TeamSpray_Spray))
  (if (not (and (not (Variables|Default|GetEnableSprayVFX)) (Components|Activation|IsActive _ns_paprika_teamspray_spray)))
    (Components|Activation|Activate _ns_paprika_teamspray_spray true))
  (Niagara|SetNiagaraVariable(Bool) (Variables|Default|GetNS_Paprika_Team_Spray_Liquid) "IsSpraying" (Variables|Default|GetEnableSprayVFX)))

(fn ConstructionScript)
