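; Gameplay ability for the Flowberry Fizz team spray (LowGrav variant), applying shields and a secondary effect while fire is held.
; A montage trigger starts a server timer that commits ammo, applies effects to the user and an effect container to selected targets, playing burst cues on alternating ticks.
; Empty ammo plays the EndSpray outro, after which the item is removed.
; Infinite ammo zeroes the cost, working around a charge-weapon counter bug.
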(event Custom|ApplyEffectIncremental
  (bind _returnvalue (Ability|GetAvatarActorfromActorInfo))
  (bind _shouldplayshieldfx (Variables|Default|GetShouldPlayShieldFX?))
  ;[0 0.5221 1] Look for targets and apply effect
  (if (Networking|IsServer)
    (bind _returnvalue_2 (Ability|CommitAbility))
    (if _returnvalue_2
      (Utilities|FlowControl|FlipFlop
        (:A
          (Variables|Default|SetShouldPlayShieldFX? true))
        (:B
          (Variables|Default|SetShouldPlayShieldFX? false)))
      (bind (_asynctaskproxy _targetdata) (Ability|Tasks|DoTargetSelection "None" "(TagName=\"Targeting.ApplicationTag.0\")" true)
        (:then
          (Fortnite|ApplyGameplayEffectsToActor _returnvalue (Utilities|Array|MakeArray "/PaprikaConsumables/Gameplay/TeamSpray/GE_Paprika_TeamSpray_Shields.GE_Paprika_TeamSpray_Shields_C" (Variables|Default|GetGESecondaryEffectToApply)) (Transformation|GetActorLocation _returnvalue) false 0 (Ability|GetActivatingAbilityComponent))
          (if _shouldplayshieldfx
            (CallFunction|ApplyShieldsCue _returnvalue)))
        (:Targeted
          (Ability|ApplyGameplayEffectContainer _targetdata "(TagName=\"Targeting.ApplicationTag.0\")")
          (for _array_element (Ability|TargetData|GetAllActorsfromTargetData _targetdata)
            (if _shouldplayshieldfx
              (CallFunction|ApplyShieldsCue _array_element)))))
      (else
        (Ability|EndAbility)))))

(event Ability|EventActivateAbility
  ;[0 0.5221 1] Activate ability and wait for spray section
  (CallFunction|CheckforInfiniteAmmo)
  (Variables|Default|SetFizzWeap (Utilities|Casting|CastToB_Paprika_TeamSpray_Weap (Ability|GetCurrentSourceWeapon)))
  (Ability|Tasks|PlayMontageandWaitforTrigger "None" (Variables|Default|GetSprayMontage)
    (:then
      (EventDispatchers|BindEventtoOnLocalAmmoChanged (Ability|GetCurrentSourceWeapon) (EventDispatchers|CreateEvent :!event Custom|OnWeaponAmmoChanged))
      (Class|BPaprikaTeamSprayWeap|ToggleSprayFX (Variables|Default|GetFizzWeap) true))
    (:Triggered
      (Variables|Default|SetStartedSpraying? true)
      (bind _returnvalue (Utilities|Time|SetTimerbyEvent (EventDispatchers|CreateEvent :!event Custom|ApplyEffectIncremental) (ScalableFloat|GetValueatLevel (Variables|Default|GetHF_TickRate)) true))
      (Variables|Default|SetApplicationTimer _returnvalue)
      (CallFunction|ApplyEffectIncremental)
      (bind _returnvalue_2 (Ability|ApplyGameplayEffectToOwner (Variables|Default|GetGE%20Spray%20Cue)))
      (Variables|Default|SetTeamSpraySpec _returnvalue_2)
      (Ability|Tasks|WaitInputRelease true
        (:OnRelease
          (Ability|EndAbility))))
    (:Cancelled
      _)
    (:Completed
      _))
  (Ability|EndAbility))

(event Ability|EventOnEndAbility
  (params (bWasCancelled :type Boolean))
  (bind _returnvalue (Ability|GetCurrentSourceWeapon))
  ;[1 0.1038 0.05208] End Ability
  (Utilities|IsValid _returnvalue
    (:Is%20Valid
      (EventDispatchers|UnbindEventfromOnLocalAmmoChanged _returnvalue (EventDispatchers|CreateEvent :!event Custom|OnWeaponAmmoChanged)))
    (:Is%20Not%20Valid
      _))
  (Class|BPaprikaTeamSprayWeap|ToggleSprayFX (Variables|Default|GetFizzWeap) false)
  (Ability|RemoveGameplayEffectFromOwnerWithHandle (Variables|Default|GetTeamSpraySpec))
  (if (Variables|Default|GetStartedSpraying?)
    (Utilities|Time|ClearandInvalidateTimerbyHandle (Variables|Default|GetApplicationTimer))
    (bind _endspraymontagehandle (Variables|Default|GetEndSprayMontageHandle)
      (:then
        (GameplayTasks|EndTask _endspraymontagehandle))
      (:else
        _))
    (if (Variables|Default|GetPlayingOutro?)
      (CallFunction|RemoveTeamSpray)
      (Ability|SendGameplayEvent (Variables|Default|GetEvent_HealConsumed))
      (else
        (Ability|Animation|MontageStop -1.0)))))

(event Custom|OnWeaponAmmoChanged
  (params (LocalCount :type Integer) (LocalRemaining :type Integer))
  ;[0.03429 0.3069 0.7899] Handle ammo change + input release
  (if (<= LocalRemaining 0)
    (bind _endspraymontagehandle (Variables|Default|GetEndSprayMontageHandle)
      (:then
        (GameplayTasks|EndTask _endspraymontagehandle))
      (:else
        _))
    (bind _asynctaskproxy (Ability|Tasks|SetNextMontageSectionandWait (Animation|GetCurrentMontage) "EndSpray")
      (:then
        (Variables|Default|SetEndSprayMontageHandle _asynctaskproxy)
        (Variables|Default|SetPlayingOutro? true))
      (:OnComplete
        (Ability|EndAbility)))))

(event Custom|ServerEndHealing
  ;[0 0.5221 1] End GE and timer on server
  (Ability|RemoveGameplayEffectFromOwnerWithHandle (Variables|Default|GetTeamSpraySpec))
  (Utilities|Time|ClearandInvalidateTimerbyHandle (Variables|Default|GetApplicationTimer)))

(event Custom|HandleRelease
  ;[0.03429 0.3069 0.7899] Handle ammo change + input release
  (Ability|HasAuthority
    (:False
      (if (not (Variables|Default|GetStartedSpraying?))
        (Utilities|FlowControl|WaitforBothExecutions true
          (:Then
            _)))
      (if (not (Variables|Default|GetPlayingOutro?))
        (Ability|EndAbility)))))

(event Custom|EnableRelease
  ;[0.03429 0.3069 0.7899] Handle ammo change + input release
  (Utilities|FlowControl|WaitforBothExecutions true
    (:Then
      (if (not (Variables|Default|GetPlayingOutro?))
        (Ability|EndAbility)))))

(fn ApplyShieldsCue
  (params (Actor :type Actor%20Object%20Reference))
  (bind (_bsuccessfullyfoundattribute _returnvalue) (Ability|Attribute|GetFloatAttribute Actor "(AttributeName=\"Shield\",Attribute=/Script/FortniteGame.FortHealthSet:Shield,AttributeOwner=\"/Script/CoreUObject.Class'/Script/FortniteGame.FortHealthSet'\")"))
  (bind (_bsuccessfullyfoundattribute_2 _returnvalue_2) (Ability|Attribute|GetFloatAttribute Actor "(AttributeName=\"CurrentShield\",Attribute=/Script/FortniteGame.FortHealthSet:CurrentShield,AttributeOwner=\"/Script/CoreUObject.Class'/Script/FortniteGame.FortHealthSet'\")"))
  (bind _returnvalue_3 (Ability|GameplayCue|MakeGameplayCueParameters 0.0))
  (if (< _returnvalue_2 _returnvalue)
    (GameplayCue|ExecuteGameplayCueOnActor(Burst) Actor (Variables|Default|GetGC_ShieldsHeal) _returnvalue_3)
    (GameplayCue|ExecuteGameplayCueOnActor(Burst) Actor (Variables|Default|GetGC_FizzApplied) _returnvalue_3))
  (return))

(fn CheckForInfiniteAmmo
  (bind _returnvalue (Ability|GetCurrentSourceWeapon))
  ;[0.4536 1 0.3115 0.5] Flowberry Fizz is set as a charge weapon, which causes a visual bug in Edit Mode when Infinite Consumables is changed while Flowberry Fizz is in your inventory: when fired, the infinity symbol doesn't replace the ammo counter, and the ammo cost is still paid. The ammo cost being paid is especially an issue because it means the weapon stops to reload, requiring the player to fire again to continue applying shields.
  ; This addresses both issues.
  (Utilities|IsValid _returnvalue
    (:Is%20Valid
      (bind _output_get (LocalVariables|SetbInfiniteAmmo (or (FortWeapon|Ammo|HasInfiniteUses _returnvalue) (FortWeapon|Ammo|InventoryOwnerHasInfiniteUses _returnvalue))))
      (if _output_get
        (FortItem|BroadcastonItemChanged (FortWeapon|GetInventoryItem _returnvalue) false true false false) ; Trigger the ammo count to change to/from an infinity symbol, which doesn't otherwise trigger because the cost is 0.
      ))
    (:Is%20Not%20Valid
      _))
  (Default|SetAmmoCost (Math|Conversions|ToFloat(Boolean) (not (LocalVariables|GetbInfiniteAmmo)))) ; Multiply the cost by 0 for Infinite Ammo, otherwise multiply by 1.
)

(fn ClientIsEmpty?
  (return (<= (FortWeapon|GetRemainingAmmo (Ability|GetCurrentSourceWeapon)) 5)))

(fn GetCustomAbilitySourceTransform
  (bind (_returnvalue_x _returnvalue_y _returnvalue_z) (Math|Vector|BreakVector (Transformation|GetSocketLocation (Variables|/Script/Engine.Character|GetMesh (CallFunction|GetAvatarasFortPlayerPawn)) "pelvis")))
  (return (Math|Vector|MakeVector (.x (Transformation|GetActorLocation (CallFunction|GetAvatarasFortPlayerPawn))) (.y (Transformation|GetActorLocation (CallFunction|GetAvatarasFortPlayerPawn))) _returnvalue_z) "0, 0, 0" "1.000000,1.000000,1.000000"))

(fn RemoveTeamSpray
  (bind _returnvalue (Ability|GetActivatingPawn))
  (bind _returnvalue_2 (Fortnite|GetFortPlayerControllerfromActor _returnvalue))
  (Utilities|IsValid _returnvalue_2
    (:Is%20Valid
      (Inventory|RemoveFortItemFromPlayer _returnvalue_2 (FortWeapon|GetInventoryItem (Ability|GetCurrentSourceWeapon)) 1 true))
    (:Is%20Not%20Valid
      (bind _asfort_athena_aibot_controller (Utilities|Casting|CastToFortAthenaAIBotController (Pawn|GetController _returnvalue)))
      (Inventory|RemoveFortItemFromPlayer _asfort_athena_aibot_controller (FortWeapon|GetInventoryItem (Ability|GetCurrentSourceWeapon)) 1 true))))
